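# Physics asset for the ponytail hair of the female medium Shapeshifter Medusa player character.
# A chain of seven capsule bodies hangs from a small driver body constrained to the head, with tight twist limits and swing limits that are nearly rigid at the root and wider further down.
# Capsules on the head, neck, upper spine and a two-capsule spine collision bone form the body volume, joined by constraints with free angular motion.
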PhysicsAsset /Game/Characters/Player/Female/Medium/Bodies/F_MED_Shapeshifter_Medusa/Meshes/Parts/F_MED_Shapeshifter_Hair_Physics
bodies {
 dyn_driver_ponytail_01 [{shapeType=Capsule, length=0.505, radius=0.505, rotation=<rot 0 -90 0>}]
 dyn_ponytail_01 [{shapeType=Capsule, length=2.5, radius=5, center=<0.001 0 0>, rotation=<rot 90 0 0>}]
 dyn_ponytail_02 [{shapeType=Capsule, length=2.75, radius=4.75, center=<0.01 0 0>, rotation=<rot 90 0 0>}]
 dyn_ponytail_03 [{shapeType=Capsule, length=3.25, radius=4.25, center=<1.2 0 0>, rotation=<rot 90 0 0>}]
 dyn_ponytail_04 [{shapeType=Capsule, length=3.25, radius=4.25, center=<1.3 -0.45 0>, rotation=<rot 90 0 0>}]
 dyn_ponytail_05 [{shapeType=Capsule, length=3.5, radius=4.25, center=<1.2 -0.6 0>, rotation=<rot 90 0 0>}]
 dyn_ponytail_06 [{shapeType=Capsule, length=2.5, radius=3.75, center=<1.5 -0.7 0>, rotation=<rot 90 0 0>}]
 dyn_ponytail_07 [{shapeType=Capsule, length=2.25, radius=2, center=<2 -1.2 0>, rotation=<rot 90 0 0>}]
 head [{shapeType=Capsule, length=8.5, radius=8.75, center=<3.092 1.9149 0.0001>, rotation=<rot 90 -90 -71.2686>}]
 neck_01 [{shapeType=Capsule, length=3.75, radius=6.505, center=<3.0677 4.9226 0>, rotation=<rot 45 90 90>}]
 spine_05 [{shapeType=Capsule, length=12, radius=17.755, center=<1.3966 8.6142 0>, rotation=<rot 0 -90 0>}]
 spine_collision [
  {shapeType=Capsule, length=10.25, radius=15.255, center=<1.8184 -1.7451 0.0177>, rotation=<rot 0.1185 -81.654 179.6243>}
  {shapeType=Capsule, length=17.75, radius=7.755, center=<-0.7299 -4.7659 -0.088>, rotation=<rot -88.906 78.2049 20.1427>}
 ]
}
constraints [
 {
  bone1Name=dyn_driver_ponytail_01
  bone2Name=head
  swing1LimitDegrees=45
  swing1Motion=Free
  swing2LimitDegrees=45
  swing2Motion=Free
  twistLimitDegrees=45
  twistMotion=Free
 }
 {bone1Name=dyn_ponytail_01, bone2Name=dyn_driver_ponytail_01, swing1LimitDegrees=1, swing2LimitDegrees=1, twistLimitDegrees=3}
 {bone1Name=dyn_ponytail_02, bone2Name=dyn_ponytail_01, swing1LimitDegrees=20, swing2LimitDegrees=20, twistLimitDegrees=3}
 {bone1Name=dyn_ponytail_03, bone2Name=dyn_ponytail_02, swing1LimitDegrees=30, swing2LimitDegrees=30, twistLimitDegrees=3}
 {bone1Name=dyn_ponytail_04, bone2Name=dyn_ponytail_03, swing1LimitDegrees=30, swing2LimitDegrees=30, twistLimitDegrees=3}
 {bone1Name=dyn_ponytail_05, bone2Name=dyn_ponytail_04, swing1LimitDegrees=45, swing2LimitDegrees=35, twistLimitDegrees=3}
 {bone1Name=dyn_ponytail_06, bone2Name=dyn_ponytail_05, swing1LimitDegrees=30, swing2LimitDegrees=30, twistLimitDegrees=3}
 {bone1Name=dyn_ponytail_07, bone2Name=dyn_ponytail_06, swing1LimitDegrees=35, swing2LimitDegrees=35, twistLimitDegrees=3}
 {
  bone1Name=head
  bone2Name=neck_01
  swing1LimitDegrees=45
  swing1Motion=Free
  swing2LimitDegrees=45
  swing2Motion=Free
  twistLimitDegrees=45
  twistMotion=Free
 }
 {
  bone1Name=neck_01
  bone2Name=spine_05
  swing1LimitDegrees=45
  swing1Motion=Free
  swing2LimitDegrees=45
  swing2Motion=Free
  twistLimitDegrees=45
  twistMotion=Free
 }
 {
  bone1Name=spine_collision
  bone2Name=spine_05
  swing1LimitDegrees=45
  swing1Motion=Free
  swing2LimitDegrees=45
  swing2Motion=Free
  twistLimitDegrees=45
  twistMotion=Free
 }
]
