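# Physics asset for the Medusa shapeshifter body of the medium female player character, covering the head and its snake hair.
# Capsule bodies run along nine snake bone chains (A through I), with most starting partway along their chain, plus a small capsule on the head.
# Constraints link each segment to the previous one and each chain's first body to the head.
# Swing and twist limits are very tight, with snake A allowed slightly more twist.
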PhysicsAsset /Game/Characters/Player/Female/Medium/Bodies/F_MED_Shapeshifter_Medusa/Meshes/Parts/F_MED_Shapeshifter_Medusa_Physics
bodies {
 head [{shapeType=Capsule, length=0.505, radius=0.505, rotation=<rot 0 -90 0>}]
 snakeA_02 [{shapeType=Capsule, length=1, radius=1.25, rotation=<rot -90 0 0>}]
 snakeA_03 [{shapeType=Capsule, length=1, radius=1.25, rotation=<rot -90 0 0>}]
 snakeA_04 [{shapeType=Capsule, length=1, radius=1.25, rotation=<rot -90 0 0>}]
 snakeA_05 [{shapeType=Capsule, length=1, radius=1.25, rotation=<rot -90 0 0>}]
 snakeA_06 [{shapeType=Capsule, length=1, radius=1.25, rotation=<rot -90 0 0>}]
 snakeA_07 [{shapeType=Capsule, length=1, radius=1.25, rotation=<rot -90 0 0>}]
 snakeA_08 [{shapeType=Capsule, length=1, radius=1.25, rotation=<rot -90 0 0>}]
 snakeA_09 [{shapeType=Capsule, length=1, radius=1.25, rotation=<rot -90 0 0>}]
 snakeB_02 [{shapeType=Capsule, length=1.5, radius=1.5, rotation=<rot -90 0 0>}]
 snakeB_03 [{shapeType=Capsule, length=1.5, radius=1.5, rotation=<rot -90 0 0>}]
 snakeB_04 [{shapeType=Capsule, length=1.5, radius=1.5, rotation=<rot -90 0 0>}]
 snakeB_05 [{shapeType=Capsule, length=1.5, radius=1.5, rotation=<rot -90 0 0>}]
 snakeB_06 [{shapeType=Capsule, length=1.5, radius=1.5, rotation=<rot -90 0 0>}]
 snakeB_07 [{shapeType=Capsule, length=1.5, radius=1.5, rotation=<rot -90 0 0>}]
 snakeB_08 [{shapeType=Capsule, length=1.5, radius=1.5, rotation=<rot -90 0 0>}]
 snakeB_09 [{shapeType=Capsule, length=1.5, radius=1.5, rotation=<rot -90 0 0>}]
 snakeC_06 [{shapeType=Capsule, length=1.75, radius=1.25, rotation=<rot -90 0 0>}]
 snakeC_07 [{shapeType=Capsule, length=1.75, radius=1.25, rotation=<rot -90 0 0>}]
 snakeC_08 [{shapeType=Capsule, length=1.75, radius=1.25, rotation=<rot -90 0 0>}]
 snakeC_09 [{shapeType=Capsule, length=1.75, radius=1.25, rotation=<rot -90 0 0>}]
 snakeC_10 [{shapeType=Capsule, length=1.75, radius=1.25, rotation=<rot -90 0 0>}]
 snakeC_11 [{shapeType=Capsule, length=1.75, radius=1.25, rotation=<rot -90 0 0>}]
 snakeD_05 [{shapeType=Capsule, length=1.75, radius=1.25, rotation=<rot -90 0 0>}]
 snakeD_06 [{shapeType=Capsule, length=1.75, radius=1.25, rotation=<rot -90 0 0>}]
 snakeD_07 [{shapeType=Capsule, length=1.75, radius=1.25, rotation=<rot -90 0 0>}]
 snakeD_08 [{shapeType=Capsule, length=1.75, radius=1.25, rotation=<rot -90 0 0>}]
 snakeD_09 [{shapeType=Capsule, length=1.75, radius=1.25, rotation=<rot -90 0 0>}]
 snakeE_04 [{shapeType=Capsule, length=1.75, radius=1.25, rotation=<rot -90 0 0>}]
 snakeE_05 [{shapeType=Capsule, length=1.75, radius=1.25, rotation=<rot -90 0 0>}]
 snakeE_06 [{shapeType=Capsule, length=1.75, radius=1.25, rotation=<rot -90 0 0>}]
 snakeE_07 [{shapeType=Capsule, length=1.75, radius=1.25, rotation=<rot -90 0 0>}]
 snakeF_04 [{shapeType=Capsule, length=1.75, radius=1.25, rotation=<rot -90 0 0>}]
 snakeF_05 [{shapeType=Capsule, length=1.75, radius=1.25, rotation=<rot -90 0 0>}]
 snakeF_06 [{shapeType=Capsule, length=1.75, radius=1.25, rotation=<rot -90 0 0>}]
 snakeF_07 [{shapeType=Capsule, length=1.75, radius=1.25, rotation=<rot -90 0 0>}]
 snakeF_08 [{shapeType=Capsule, length=1.75, radius=1.25, rotation=<rot -90 0 0>}]
 snakeF_09 [{shapeType=Capsule, length=1.75, radius=1.25, rotation=<rot -90 0 0>}]
 snakeG_05 [{shapeType=Capsule, length=1.75, radius=1.25, rotation=<rot -90 0 0>}]
 snakeG_06 [{shapeType=Capsule, length=1.75, radius=1.25, rotation=<rot -90 0 0>}]
 snakeG_07 [{shapeType=Capsule, length=1.75, radius=1.25, rotation=<rot -90 0 0>}]
 snakeG_08 [{shapeType=Capsule, length=1.75, radius=1.25, rotation=<rot -90 0 0>}]
 snakeG_09 [{shapeType=Capsule, length=1.75, radius=1.25, rotation=<rot -90 0 0>}]
 snakeG_10 [{shapeType=Capsule, length=1.75, radius=1.25, rotation=<rot -90 0 0>}]
 snakeG_11 [{shapeType=Capsule, length=1.75, radius=1.25, rotation=<rot -90 0 0>}]
 snakeH_07 [{shapeType=Capsule, length=1.75, radius=1.25, rotation=<rot -90 0 0>}]
 snakeH_08 [{shapeType=Capsule, length=1.75, radius=1.25, rotation=<rot -90 0 0>}]
 snakeH_09 [{shapeType=Capsule, length=1.75, radius=1.25, rotation=<rot -90 0 0>}]
 snakeH_10 [{shapeType=Capsule, length=1.75, radius=1.25, rotation=<rot -90 0 0>}]
 snakeH_11 [{shapeType=Capsule, length=1.75, radius=1.25, rotation=<rot -90 0 0>}]
 snakeH_12 [{shapeType=Capsule, length=1.75, radius=1.25, rotation=<rot -90 0 0>}]
 snakeH_13 [{shapeType=Capsule, length=1.75, radius=1.25, rotation=<rot -90 0 0>}]
 snakeI_05 [{shapeType=Capsule, length=1.75, radius=1.25, rotation=<rot -90 0 0>}]
 snakeI_06 [{shapeType=Capsule, length=1.75, radius=1.25, rotation=<rot -90 0 0>}]
 snakeI_07 [{shapeType=Capsule, length=1.75, radius=1.25, rotation=<rot -90 0 0>}]
 snakeI_08 [{shapeType=Capsule, length=1.75, radius=1.25, rotation=<rot -90 0 0>}]
 snakeI_09 [{shapeType=Capsule, length=1.75, radius=1.25, rotation=<rot -90 0 0>}]
}
constraints [
 {bone1Name=snakeA_02, bone2Name=head, swing1LimitDegrees=1, swing2LimitDegrees=1, twistLimitDegrees=5}
 {bone1Name=snakeA_03, bone2Name=snakeA_02, swing1LimitDegrees=1, swing2LimitDegrees=1, twistLimitDegrees=5}
 {bone1Name=snakeA_04, bone2Name=snakeA_03, swing1LimitDegrees=1, swing2LimitDegrees=1, twistLimitDegrees=5}
 {bone1Name=snakeA_05, bone2Name=snakeA_04, swing1LimitDegrees=1, swing2LimitDegrees=1, twistLimitDegrees=5}
 {bone1Name=snakeA_06, bone2Name=snakeA_05, swing1LimitDegrees=1, swing2LimitDegrees=1, twistLimitDegrees=5}
 {bone1Name=snakeA_07, bone2Name=snakeA_06, swing1LimitDegrees=1, swing2LimitDegrees=1, twistLimitDegrees=5}
 {bone1Name=snakeA_08, bone2Name=snakeA_07, swing1LimitDegrees=1, swing2LimitDegrees=1, twistLimitDegrees=5}
 {bone1Name=snakeA_09, bone2Name=snakeA_08, swing1LimitDegrees=1, swing2LimitDegrees=1, twistLimitDegrees=5}
 {bone1Name=snakeB_02, bone2Name=head, swing1LimitDegrees=1, swing2LimitDegrees=1, twistLimitDegrees=0.1}
 {bone1Name=snakeB_03, bone2Name=snakeB_02, swing1LimitDegrees=1, swing2LimitDegrees=1, twistLimitDegrees=0.1}
 {bone1Name=snakeB_04, bone2Name=snakeB_03, swing1LimitDegrees=1, swing2LimitDegrees=1, twistLimitDegrees=0.1}
 {bone1Name=snakeB_05, bone2Name=snakeB_04, swing1LimitDegrees=1, swing2LimitDegrees=1, twistLimitDegrees=0.1}
 {bone1Name=snakeB_06, bone2Name=snakeB_05, swing1LimitDegrees=1, swing2LimitDegrees=1, twistLimitDegrees=0.1}
 {bone1Name=snakeB_07, bone2Name=snakeB_06, swing1LimitDegrees=1, swing2LimitDegrees=1, twistLimitDegrees=0.1}
 {bone1Name=snakeB_08, bone2Name=snakeB_07, swing1LimitDegrees=1, swing2LimitDegrees=1, twistLimitDegrees=0.1}
 {bone1Name=snakeB_09, bone2Name=snakeB_08, swing1LimitDegrees=1, swing2LimitDegrees=1, twistLimitDegrees=0.1}
 {bone1Name=snakeC_06, bone2Name=head, swing1LimitDegrees=1, swing2LimitDegrees=1, twistLimitDegrees=0.1}
 {bone1Name=snakeC_07, bone2Name=snakeC_06, swing1LimitDegrees=1, swing2LimitDegrees=1, twistLimitDegrees=0.1}
 {bone1Name=snakeC_08, bone2Name=snakeC_07, swing1LimitDegrees=1, swing2LimitDegrees=1, twistLimitDegrees=0.1}
 {bone1Name=snakeC_09, bone2Name=snakeC_08, swing1LimitDegrees=1, swing2LimitDegrees=1, twistLimitDegrees=0.1}
 {bone1Name=snakeC_10, bone2Name=snakeC_09, swing1LimitDegrees=1, swing2LimitDegrees=1, twistLimitDegrees=0.1}
 {bone1Name=snakeC_11, bone2Name=snakeC_10, swing1LimitDegrees=1, swing2LimitDegrees=1, twistLimitDegrees=0.1}
 {bone1Name=snakeD_05, bone2Name=head, swing1LimitDegrees=1, swing2LimitDegrees=1, twistLimitDegrees=0.1}
 {bone1Name=snakeD_06, bone2Name=snakeD_05, swing1LimitDegrees=1, swing2LimitDegrees=1, twistLimitDegrees=0.1}
 {bone1Name=snakeD_07, bone2Name=snakeD_06, swing1LimitDegrees=1, swing2LimitDegrees=1, twistLimitDegrees=0.1}
 {bone1Name=snakeD_08, bone2Name=snakeD_07, swing1LimitDegrees=1, swing2LimitDegrees=1, twistLimitDegrees=0.1}
 {bone1Name=snakeD_09, bone2Name=snakeD_08, swing1LimitDegrees=1, swing2LimitDegrees=1, twistLimitDegrees=0.1}
 {bone1Name=snakeE_04, bone2Name=head, swing1LimitDegrees=1, swing2LimitDegrees=1, twistLimitDegrees=0.1}
 {bone1Name=snakeE_05, bone2Name=snakeE_04, swing1LimitDegrees=1, swing2LimitDegrees=1, twistLimitDegrees=0.1}
 {bone1Name=snakeE_06, bone2Name=snakeE_05, swing1LimitDegrees=1, swing2LimitDegrees=1, twistLimitDegrees=0.1}
 {bone1Name=snakeE_07, bone2Name=snakeE_06, swing1LimitDegrees=1, swing2LimitDegrees=1, twistLimitDegrees=0.1}
 {bone1Name=snakeF_04, bone2Name=head, swing1LimitDegrees=1, swing2LimitDegrees=1, twistLimitDegrees=0.1}
 {bone1Name=snakeF_05, bone2Name=snakeF_04, swing1LimitDegrees=1, swing2LimitDegrees=1, twistLimitDegrees=0.1}
 {bone1Name=snakeF_06, bone2Name=snakeF_05, swing1LimitDegrees=1, swing2LimitDegrees=1, twistLimitDegrees=0.1}
 {bone1Name=snakeF_07, bone2Name=snakeF_06, swing1LimitDegrees=1, swing2LimitDegrees=1, twistLimitDegrees=0.1}
 {bone1Name=snakeF_08, bone2Name=snakeF_07, swing1LimitDegrees=1, swing2LimitDegrees=1, twistLimitDegrees=0.1}
 {bone1Name=snakeF_09, bone2Name=snakeF_08, swing1LimitDegrees=1, swing2LimitDegrees=1, twistLimitDegrees=0.1}
 {bone1Name=snakeG_05, bone2Name=head, swing1LimitDegrees=1, swing2LimitDegrees=1, twistLimitDegrees=0.1}
 {bone1Name=snakeG_06, bone2Name=snakeG_05, swing1LimitDegrees=1, swing2LimitDegrees=1, twistLimitDegrees=0.1}
 {bone1Name=snakeG_07, bone2Name=snakeG_06, swing1LimitDegrees=1, swing2LimitDegrees=1, twistLimitDegrees=0.1}
 {bone1Name=snakeG_08, bone2Name=snakeG_07, swing1LimitDegrees=1, swing2LimitDegrees=1, twistLimitDegrees=0.1}
 {bone1Name=snakeG_09, bone2Name=snakeG_08, swing1LimitDegrees=1, swing2LimitDegrees=1, twistLimitDegrees=0.1}
 {bone1Name=snakeG_10, bone2Name=snakeG_09, swing1LimitDegrees=1, swing2LimitDegrees=1, twistLimitDegrees=0.1}
 {bone1Name=snakeG_11, bone2Name=snakeG_10, swing1LimitDegrees=1, swing2LimitDegrees=1, twistLimitDegrees=0.1}
 {bone1Name=snakeH_07, bone2Name=head, swing1LimitDegrees=1, swing2LimitDegrees=1, twistLimitDegrees=0.1}
 {bone1Name=snakeH_08, bone2Name=snakeH_07, swing1LimitDegrees=1, swing2LimitDegrees=1, twistLimitDegrees=0.1}
 {bone1Name=snakeH_09, bone2Name=snakeH_08, swing1LimitDegrees=1, swing2LimitDegrees=1, twistLimitDegrees=0.1}
 {bone1Name=snakeH_10, bone2Name=snakeH_09, swing1LimitDegrees=1, swing2LimitDegrees=1, twistLimitDegrees=0.1}
 {bone1Name=snakeH_11, bone2Name=snakeH_10, swing1LimitDegrees=1, swing2LimitDegrees=1, twistLimitDegrees=0.1}
 {bone1Name=snakeH_12, bone2Name=snakeH_11, swing1LimitDegrees=1, swing2LimitDegrees=1, twistLimitDegrees=0.1}
 {bone1Name=snakeH_13, bone2Name=snakeH_12, swing1LimitDegrees=1, swing2LimitDegrees=1, twistLimitDegrees=0.1}
 {bone1Name=snakeI_05, bone2Name=head, swing1LimitDegrees=1, swing2LimitDegrees=1, twistLimitDegrees=0.1}
 {bone1Name=snakeI_06, bone2Name=snakeI_05, swing1LimitDegrees=1, swing2LimitDegrees=1, twistLimitDegrees=0.1}
 {bone1Name=snakeI_07, bone2Name=snakeI_06, swing1LimitDegrees=1, swing2LimitDegrees=1, twistLimitDegrees=0.1}
 {bone1Name=snakeI_08, bone2Name=snakeI_07, swing1LimitDegrees=1, swing2LimitDegrees=1, twistLimitDegrees=0.1}
 {bone1Name=snakeI_09, bone2Name=snakeI_08, swing1LimitDegrees=1, swing2LimitDegrees=1, twistLimitDegrees=0.1}
]
