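# Animation Blueprint for the Medusa Shapeshifter body part, animating and simulating its hair and head snakes.
# It copies the parent mesh pose, layers ADS additives, offsets spine_collision for back bling, and scales hair, band and snake bones from variables.
# A time-driven scale additive then feeds two RigidBody simulations, the first using state-based gravity from a deprecated helper.
# Events set back-bling, SnakesOff and variant-gated preview flags, with a leftover "Huzzah" debug print.
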AnimBlueprint /Game/Characters/Player/Female/Medium/Bodies/F_MED_Shapeshifter_Medusa/Meshes/Parts/F_MED_Shapeshifter_Medusa_AnimBP
target_skeleton=&/F_MED_Shapeshifter_Medusa_Skeleton
graphs [
 AnimGraph {nodes [
   n1 = Apply Additive(
    Base=Copy Pose From Mesh() {bUseAttachedParent=true}
    Additive="Sequence Player (additive) 'Shapeshifter_ADS_Loop'"() {sequence=&/Shapeshifter_ADS_Loop}
   ) {alphaBoolBlend {blendOption=Linear}}
   n2 = Blendspace Player 'Shapeshifter_ADS_DirPose_BS'(X=Get PawnMovementDirectionAzimuth2() -> (PawnMovementDirectionAzimuth2:float), Y=Get PawnSpeed2D()) {blendSpace=&/Shapeshifter_ADS_DirPose_BS}
   n3 = "Transform (Modify) Bone - Bone: spine_collision"(
    ComponentPose=Local To Component(LocalPose=Apply Additive(Base=@n1, Additive=@n2) {alphaBoolBlend {blendOption=Linear}})
    Alpha=Get BackBlingAlpha() -> (BackBlingAlpha:float)
    Translation="0,40,0"
    Rotation="0,0,0"
    Scale="1,1,1"
   ) {
    translationMode=BMM_Additive
    translationSpace=BCS_BoneSpace
    alphaBoolBlend {blendOption=Linear}
    boneToModify {boneName=spine_collision}
    translation {y=40}
   }
   n4 = "Transform (Modify) Bone - Bone: dyn_head_band"(ComponentPose=@n3, Alpha=1, Translation="0,0,0", Rotation="0,0,0", Scale=Get SnakeBandScale() -> (SnakeBandScale:Vector)) {
    scaleMode=BMM_Replace
    scaleSpace=BCS_BoneSpace
    alphaBoolBlend {blendOption=Linear}
    alphaScaleBiasClamp {bInterpResult=true, interpSpeedDecreasing=5, interpSpeedIncreasing=5}
    boneToModify {boneName=dyn_head_band}
   }
   n5 = "Transform (Modify) Bone - Bone: dyn_hair_band"(
    ComponentPose="Transform (Modify) Bone - Bone: dyn_ponytail_01"(ComponentPose=@n4, Alpha=1, Translation="0,0,0", Rotation="0,0,0", Scale=Get HairScaleInterp() -> (HairScaleInterp:Vector)) {scaleMode=BMM_Replace, alphaBoolBlend {blendOption=Linear}, boneToModify {boneName=dyn_ponytail_01}}
    Alpha=1
    Translation="0,0,0"
    Rotation="0,0,0"
    Scale=Get HairBandScale() -> (HairBandScale:Vector)
   ) {
    scaleMode=BMM_Replace
    scaleSpace=BCS_BoneSpace
    alphaBoolBlend {blendOption=Linear}
    alphaScaleBiasClamp {bInterpResult=true, interpSpeedDecreasing=5, interpSpeedIncreasing=5}
    boneToModify {boneName=dyn_hair_band}
   }
   n6 = Get SnakeHairScaleInterp() -> (SnakeHairScaleInterp:Vector)
   n7 = "Transform (Modify) Bone - Bone: snakeA_01"(ComponentPose=@n5, Alpha=1, Translation="0,0,0", Rotation="0,0,0", Scale=@n6) {scaleMode=BMM_Replace, alphaBoolBlend {blendOption=Linear}, boneToModify {boneName=snakeA_01}}
   n8 = "Transform (Modify) Bone - Bone: snakeD_01"(
    ComponentPose="Transform (Modify) Bone - Bone: snakeC_01"(
     ComponentPose="Transform (Modify) Bone - Bone: snakeB_01"(ComponentPose=@n7, Alpha=1, Translation="0,0,0", Rotation="0,0,0", Scale=@n6) {scaleMode=BMM_Replace, alphaBoolBlend {blendOption=Linear}, boneToModify {boneName=snakeB_01}}
     Alpha=1
     Translation="0,0,0"
     Rotation="0,0,0"
     Scale=@n6
    ) {scaleMode=BMM_Replace, alphaBoolBlend {blendOption=Linear}, boneToModify {boneName=snakeC_01}}
    Alpha=1
    Translation="0,0,0"
    Rotation="0,0,0"
    Scale=@n6
   ) {scaleMode=BMM_Replace, alphaBoolBlend {blendOption=Linear}, boneToModify {boneName=snakeD_01}}
   n9 = "Transform (Modify) Bone - Bone: snakeG_01"(
    ComponentPose="Transform (Modify) Bone - Bone: snakeF_01"(
     ComponentPose="Transform (Modify) Bone - Bone: snakeE_01"(ComponentPose=@n8, Alpha=1, Translation="0,0,0", Rotation="0,0,0", Scale=@n6) {scaleMode=BMM_Replace, alphaBoolBlend {blendOption=Linear}, boneToModify {boneName=snakeE_01}}
     Alpha=1
     Translation="0,0,0"
     Rotation="0,0,0"
     Scale=@n6
    ) {scaleMode=BMM_Replace, alphaBoolBlend {blendOption=Linear}, boneToModify {boneName=snakeF_01}}
    Alpha=1
    Translation="0,0,0"
    Rotation="0,0,0"
    Scale=@n6
   ) {scaleMode=BMM_Replace, alphaBoolBlend {blendOption=Linear}, boneToModify {boneName=snakeG_01}}
   n10 = Component To Local(ComponentPose="Transform (Modify) Bone - Bone: snakeI_01"(
     ComponentPose="Transform (Modify) Bone - Bone: snakeH_01"(ComponentPose=@n9, Alpha=1, Translation="0,0,0", Rotation="0,0,0", Scale=@n6) {scaleMode=BMM_Replace, alphaBoolBlend {blendOption=Linear}, boneToModify {boneName=snakeH_01}}
     Alpha=1
     Translation="0,0,0"
     Rotation="0,0,0"
     Scale=@n6
    ) {scaleMode=BMM_Replace, alphaBoolBlend {blendOption=Linear}, boneToModify {boneName=snakeI_01}})
   n11 = "Sequence Evaluator (by time) 'Shapeshifter_ADS_Scale'"!(ExplicitTime:float=Get ScaleAnimExplicitTime() -> (ScaleAnimExplicitTime:float), StartPosition:float=0) -> (Pose:Pose Link Structure) {
    is=title AnimGraphNode_SequenceEvaluator
    sequence=&/Shapeshifter_ADS_Scale
   }
   n12 = RigidBody!(
    ComponentPose=Local To Component(LocalPose=Apply Additive(Base=@n10, Additive=@n11) {alphaBoolBlend {blendOption=Linear}})
    Alpha=Get HairRbanAlphaInterp() -> (HairRbanAlphaInterp:float)
    OverrideWorldGravity:Vector=Get GravityOverride()
    ExternalForce
   ) -> (Pose) {
    bOverrideWorldGravity=true
    overridePhysicsAsset=&/F_MED_Shapeshifter_Hair_Physics
    simulationSpace=BaseBoneSpace
    alphaBoolBlend {blendOption=Linear}
    baseBoneRef {boneName=root}
    componentAppliedLinearAccClamp=<1100 1100 1100>
    componentLinearAccScale=<0.3 0.3 0.5>
    componentLinearVelScale=<0.3 0.3 0.5>
    overrideWorldGravity {z=-850}
   }
   n13 = RigidBody!(ComponentPose=@n12, Alpha=Get RigidsAlpha() -> (RigidsAlpha:float), ExternalForce) -> (Pose) {
    bOverrideWorldGravity=true
    overridePhysicsAsset=&/F_MED_Shapeshifter_Medusa_Physics
    simulationSpace=BaseBoneSpace
    alphaBoolBlend {blendOption=Linear}
    baseBoneRef {boneName=spine_05}
    componentAppliedLinearAccClamp=<2000 2000 1000>
    componentLinearAccScale=<0.4 0.4 0.2>
    componentLinearVelScale=<0.4 0.4 0.2>
    overrideWorldGravity {z=125}
   }
   Output Pose(Result=Component To Local(ComponentPose=@n13)) {layerGroup=None}
  ]}
 EventGraph {nodes [
   Event BlueprintUpdateAnimation()
   > n1 = Gravity Override State Function DEPRECATED(CharacterPartAnimInstance=Self-Reference(), GameTimeInSeconds=GetGameTimeInSeconds(), GravityOverrideParams=Get Gravity Override Params())
   > Set GravityOverride(GravityOverride=@n1)
   Event OnCharacterCustomizationLoaded()
   > n2 = Sequence()
   > Set BackBlingAlpha(BackBlingAlpha:float=SelectFloat(A=0, B=1, bPickA=CheckForBackbling())) -> (Output_Get:float)
   n3 = Branch(
    execute=@n2.then_1
    Condition=CharacterItemDefinitionHasAnyMetaTag(Tags=Make Array("[0]"=(TagName="Cosmetics.MetaTags.ShapeShifter.SnakesOff")))
   )
   > Set bDisableSnakes(bDisableSnakes:bool=true) -> (Output_Get:bool)
   Set bDisableSnakes(execute=@n3.else, bDisableSnakes=false)
   n4 = Event StartPreviewReactiveResponse!() -> (
    OutputDelegate
    then
    ChannelAndVariant:"Cosmetic Variant Info Structure (by ref)"
    Preview:"Fort Cosmetic Reactive Preview (F CRP) Object Reference"
   ) {is=title K2Node_Event}
   > Gate Is Variant Reactive Towards!(execute:Exec, CosmeticVariantInfo:Cosmetic Variant Info Structure=@n4.ChannelAndVariant) -> (true:Exec) {is=title K2Node_MacroInstance}
   > PrintString!(execute, InString=Huzzah, bPrintToScreen, bPrintToLog, TextColor, Duration, Key) -> (then)
   > Delay(Duration=0.5)
   > Set bIsReactivePreviewSnakesFromAds(bIsReactivePreviewSnakesFromAds:bool=true) -> (Output_Get:bool)
   n5 = Event StopPreviewReactiveResponse!() -> (
    OutputDelegate
    then
    ChannelAndVariant:"Cosmetic Variant Info Structure (by ref)"
    Preview:"Fort Cosmetic Reactive Preview (F CRP) Object Reference"
   ) {is=title K2Node_Event}
   > Gate Is Variant Reactive Towards(CosmeticVariantInfo=@n5.ChannelAndVariant)
   > Set bIsReactivePreviewSnakesFromAds(bIsReactivePreviewSnakesFromAds=false)
  ]}
 Gate_IsVariantReactiveTowards {nodes [
   n1 = Inputs!() -> (execute, CosmeticVariantInfo:Cosmetic Variant Info Structure)
   n2 = Break Cosmetic Variant Info!(CosmeticVariantInfo:"Cosmetic Variant Info Structure (by ref)"=@n1) -> (VariantChannelTag:Gameplay Tag Structure, ActiveVariantTag:Gameplay Tag Structure) {is=title K2Node_BreakStruct}
   > Branch(Condition=AND Boolean(
     A="Equal (GameplayTag)"!(A:Gameplay Tag Structure=@n2, B:Gameplay Tag Structure=(TagName="Cosmetics.Variant.Channel.Parts")) -> (ReturnValue:bool) {is=title K2Node_CallFunction}
     B="Equal (GameplayTag)"(A=@n2.ActiveVariantTag, B=(TagName="Cosmetics.Variant.Property.Stage1"))
    ))
   > Outputs!(true:Exec)
  ]}
]
variables [Gravity Override Params {
  type=GravityOverrideParamsStruct
  default {
   front_End {gravityOverride=<0 0 -850>}
   onGround_Moving {gravityOverride=<0 0 -850>}
   onGround_Standing {gravityOverride=<0 0 -850>}
   skydive_DivingDown {gravityOverride=<0 0 0>}
   skydive_FloatingDown {gravityOverride=<0 0 0>}
   skydive_GoingUp {gravityOverride=<0 0 -100>, windOffset=<0 -600 400>}
   skydive_Parachuting {gravityOverride=<0 0 -850>}
   swimming {gravityOverride=<0 0 0>}
  }
 }]
